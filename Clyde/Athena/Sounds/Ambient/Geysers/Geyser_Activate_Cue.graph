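# Sound cue for a geyser activating, playing a launch sound as part of the Athena ambient set.
# A Random node picks one of three equally weighted Geyser_Launch wave players on each play.
# It routes to the Fort_Gadgets sound class and borrows the distant golf cart movement attenuation asset for spatial falloff.
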SoundCue /Clyde/Athena/Sounds/Ambient/Geysers/Geyser_Activate_Cue
root=Random_1
nodes [
 {id=Random_1, children [WavePlayer_4,WavePlayer_5,WavePlayer_6], properties {weights [1,1,1]}}
 {id=WavePlayer_4, properties {soundWaveAssetPtr=/Game/Athena/Sounds/Ambient/Geysers/Geyser_Launch_01}}
 {id=WavePlayer_5, properties {soundWaveAssetPtr=/Game/Athena/Sounds/Ambient/Geysers/Geyser_Launch_02}}
 {id=WavePlayer_6, properties {soundWaveAssetPtr=/Game/Athena/Sounds/Ambient/Geysers/Geyser_Launch_03}}
]
properties {
 attenuationSettings=/Game/Sounds/Attenuation/Fort_GolfCart_Movement_Distant_Att
 maxDistance=10100
 soundClassObject=&Fort_Gadgets
}
