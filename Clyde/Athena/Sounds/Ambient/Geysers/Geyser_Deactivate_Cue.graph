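# Sound cue named for geyser deactivation in Athena's ambient sounds, currently empty with no root or sound nodes, so it has no audio to play.
# It still carries playback settings, using a distant golf-cart movement attenuation asset, a set maximum distance and the Fort_Gadgets sound class.
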SoundCue /Clyde/Athena/Sounds/Ambient/Geysers/Geyser_Deactivate_Cue
root=null
nodes []
properties {
 attenuationSettings=/Game/Sounds/Attenuation/Fort_GolfCart_Movement_Distant_Att
 maxDistance=10000
 soundClassObject=&Fort_Gadgets
}
